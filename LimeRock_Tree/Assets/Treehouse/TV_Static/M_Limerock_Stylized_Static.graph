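# Unlit TV screen material for the Limerock treehouse that alternates stylized static with images from a texture-array flipbook.
# Stepped-panning noise is colored through a curve atlas, and a second sample distorts the image UVs.
# A time interval cycles image slices and fades static around each brief showing, while scripted parameters can pick an image or force static.
# EyeAdaptationInverse compensates the emissive result for exposure.
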Material /LimeRock_Tree/Assets/Treehouse/TV_Static/M_Limerock_Stylized_Static
graph {
 properties {bAutomaticallySetUsageInEditor=false, shadingModel=MSM_Unlit, usedShadingModels=Unlit}
 nodes [
  n1 = Divide(A=Time(), B=ScalarParameter() {defaultValue=10, parameterName=Interval_Static})
  Ceil(None=@n1)
  n2 = TextureCoordinate()
  n3 = AppendVector(
   A=ScalarParameter() {defaultValue=1, group=Emissive Static, parameterName=Static Scale X}
   B=ScalarParameter() {defaultValue=1, group=Emissive Static, parameterName=Static Scale Y}
  )
  n4 = ScalarParameter() {defaultValue=1, group=Emissive Static, parameterName=Static Offset X}
  n5 = ScalarParameter() {defaultValue=1, group=Emissive Static, parameterName=Static Offset Y}
  n6 = ScalarParameter() {defaultValue=1, group=Emissive Static, parameterName=Static Offset Step Jump X}
  n7 = ScalarParameter() {group=Emissive Static, parameterName=Static Offset Step Jump Y}
  n8 = ScalarParameter() {defaultValue=5, group=Emissive Static, parameterName=Static Offset Step Smoothness X}
  n9 = ScalarParameter() {group=Emissive Static, parameterName=Static Offset Step Smoothness Y}
  n10 = Add(
   A=Add(A=Multiply(A=@n2, B=@n3), B=AppendVector(A=@n4, B=@n5))
   B=MaterialFunctionCall!(
    TimeSpeed=ScalarParameter() {defaultValue=1, group=Emissive Static, parameterName=Step Time Speed}
    Jump Amount Per Step=AppendVector(A=@n6, B=@n7)
    Number of Steps Per Unit of Time=AppendVector(A=@n8, B=@n9)
    Time Offset
   ) {materialFunction=/Engine/Functions/Engine_MaterialFunctions02/SteppingPannerTime}.Result
  )
  n11 = DDX(Value=@n2)
  n12 = DDY(Value=@n2)
  n13 = TextureSample!(UVs=@n10, Tex, "DDX(UVs)"=@n11, "DDY(UVs)"=@n12, Apply View MipBias) {automaticViewMipBias=false, mipValueMode=TMVM_Derivative, texture=&/TX_LimeRock_Static_01}
  n14 = CurveAtlasRowParameter(CurveTime=@n13.R) {
   atlas=&/CA_LimeRock_Static
   curve=&/C_LimeRock_Static
   group=Emissive Static
   parameterName=Emissive Curve
  }
  n15 = Multiply(A=@n14, B=ScalarParameter() {defaultValue=1, group=Emissive Static, parameterName=Emissive Strength})
  n16 = TextureSample(UVs=Add(A=@n10) {constB=0.5}, "DDX(UVs)"=@n11, "DDY(UVs)"=@n12) {automaticViewMipBias=false, mipValueMode=TMVM_Derivative, texture=&TX_LimeRock_Static_01}
  n17 = Multiply(A=Subtract(A=@n16.R) {constB=0.5}, B=ScalarParameter() {defaultValue=0.016, parameterName=StaticDistortion_A})
  n18 = ScalarParameter() {parameterName=TVImage_UVOffset_X}
  n19 = ScalarParameter() {defaultValue=-0.0976, parameterName=TVImage_UVOffset_Y}
  n20 = MaterialFunctionCall(
   UVs=Add(A=TextureCoordinate(), B=AppendVector(A=@n18, B=@n19))
   Texture Scale=AppendVector(
    A=ScalarParameter() {defaultValue=0.9752, parameterName=TVImage_UVScale_X}
    B=ScalarParameter() {defaultValue=0.7425, parameterName=TVImage_UVScale_Y}
   )
  ) {materialFunction=&ScaleUVsByCenter}
  n21 = Floor(None=@n1)
  n22 = ScalarParameter() {defaultValue=8, parameterName=NumberofImages}
  n23 = Multiply(A=Frac(None=Divide(A=@n21, B=@n22)), B=@n22)
  n24 = ScalarParameter() {parameterName=Scripted_TVImageSelect}
  n25 = ScalarParameter() {parameterName=Scripted_EnableImageSelect}
  n26 = TextureSample!(UVs=AppendVector(A=Add(A=@n17, B=@n20.UVs), B=LinearInterpolate(A=@n23, B=@n24, Alpha=@n25)), Tex, Apply View MipBias) {texture=&/LimeRockTV_FlipbookArray}
  n27 = Divide(A=Frac(None=@n1), B=ScalarParameter() {defaultValue=0.25, parameterName=Interval_Image_LenthPercentage})
  n28 = Frac(None=Subtract(A=Saturate(None=@n27)) {constB=0.0001})
  n29 = ScalarParameter() {defaultValue=0.07, parameterName=StaticFadeTime}
  n30 = Divide(A=@n28, B=@n29)
  n31 = Divide(A=Subtract(A=OneMinus(None=@n28)) {constB=0.5}, B=@n29)
  n32 = OneMinus(None=Multiply(A=Saturate(None=@n30), B=Saturate(None=@n31)))
  n33 = Saturate(None=Max(A=Multiply(A=@n13.R, B=@n32), B=ScalarParameter() {defaultValue=0.09, parameterName=Static_Min}))
  n34 = Saturate(None=OneMinus(None=ScalarParameter() {parameterName=Scripted_ForceStatic}))
  n35 = OneMinus(None=Round(None=@n28))
  EyeAdaptationInverse(LightValueInput=LinearInterpolate(A=@n15, B=LinearInterpolate(A=@n26.RGB, B=@n15, Alpha=@n33), Alpha=Multiply(A=@n34, B=@n35)))
  ScalarParameter() {parameterName=MIP Bias}
 ]
}
